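# Looping Niagara effect for the FeatherMud emote: a spotlight with falling confetti and magical puff clouds.
# Confetti spawns in an overhead cone, animating sub-images under drag, gravity and curl noise, while tinted puffs fade in and out low down.
# The burst-spawned mesh Spotlight outlives its lifetime, with alpha driven by User.NormalizedNotifyProgress.
# Time of day tints confetti and spotlight, and front-end flags alter the spotlight's day color.
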NiagaraSystem /BRCosmetics/Animation/Game/MainPlayer/Emotes/FeatherMud/FX/NS_Emote_FeatherMud_Looping
dependencies {
 usedDataInterfaces [&NiagaraDataInterfaceArrayColor, &NiagaraDataInterfaceCurve, &NiagaraDataInterfaceVectorCurve]
 usedDynamicInputs []
 usedModules [&SystemState]
}
emitters {
 Confetti {
  emitterUpdateScript [
   &EmitterState(Delay First Loop Only=true, Life Cycle Mode=::NewEnumerator1, Loop Delay=1, Loop Duration=1, UseLoopDelay=true)
   &SpawnRate(SpawnRate=40, Use Spawn Probability=false)
  ]
  particleSpawnScript [
   &InitializeParticle(
    Color=&SelectLinearColorFromArray(
     Array Sampling Mode=::NewEnumerator0
     Direct Array Index=?
     Color Selection Array=DataInterface() {propertyValues {ColorData [<rgba 2 1.2806 0.1957 1>, <rgba 5 4.4057 3.6132 1>, <rgba 2 1.3435 0.9978 1>]}}
     Random Seed=0
    )
    Lifetime Max=1
    Lifetime Min=0.6
    Lifetime Mode=::NewEnumerator1
    Sprite Size Mode=::NewEnumerator1
    Uniform Sprite Size Max=6
    Uniform Sprite Size Min=3
    UsePositionOffset=false
    Write Material Random=true
   )
   &ShapeLocation(
    Cone Angle=55
    Cone Inner Angle=20
    Cone Length=120
    MaskSpawnGroup=false
    Non Uniform Scale=<1 1 -1>
    Offset=<0 0 130>
    Offset Mode=::NewEnumerator0
    Override Seed=false
    Shape Primitive=::NewEnumerator5
    UseOwnerScale=true
   )
   &Color(Color=&MultiplyLinearColor(
     Linear Color=$Particles.Color
     Scale Factor=&MakeVector4_FromLinearColor(LinearColor=&FNTimeOfDayColor(
       Night Color=<rgba 0.2706 0.2383 0.4948 1>
       Morning Color=<rgba 1 0.7262 0.6963 1>
       Day Color=<rgba 1 0.9315 0.9012 1>
       Evening Color=<rgba 1 0.9069 0.7687 1>
      ))
    ))
  ]
  particleUpdateScript [
   &Drag(Drag=3.5, Use Linear Drag=true, Use Rotational Drag=true)
   &GravityForce(Gravity=&RandomRangeVector(Minimum=<0 0 -220>, Maximum=<0 0 -400>))
   &CurlNoiseForce(
    Bias Noise=false
    Noise Frequency=10
    Noise Strength=&RandomRangeFloat(Minimum=30, Maximum=50)
    Pan Noise=true
    Pan Noise Field=<0.1 0.1 0.1>
   )
   &SolveForcesAndVelocity(Clamp Velocity=false, Limit Acceleration=false)
   &ParticleState()
   &ScaleSpriteSize(Uniform Curve Sprite Scale=DataInterface() {propertyValues {Curve {keys [
        {interpMode=RCIM_Cubic, leaveTangent=12.0846, tangentMode=RCTM_Break, time=0, value=0}
        {interpMode=RCIM_Cubic, tangentMode=RCTM_Break, time=0.1, value=1}
        {arriveTangent=-2.7226, tangentMode=RCTM_User, time=1, value=0}
       ]}}})
   /SetVariables(Particles.SubImageIndex=&FloatFromCurve(FloatCurve=DataInterface() {propertyValues {Curve {keys [<at 0 0>,<at 1 1>]}}}, Scale Curve=36)) {isSetParameters=true}
  ]
  rendererClasses [&NiagaraSpriteRendererProperties]
  renderers [{rendererClass=&NiagaraSpriteRendererProperties}]
 }
 MagicalPuffCloud {
  emitterUpdateScript [
   &EmitterState()
   &SpawnRate(SpawnRate=10, Use Spawn Probability=false)
  ]
  particleSpawnScript [
   &InitializeParticle(
    Color Maximum=<rgba 0.7296 1 0.9295 1>
    Color Minimum=<rgba 1 0.6642 0.1083 1>
    Color Mode=::NewEnumerator2
    Lifetime Max=2.5
    Lifetime Mode=::NewEnumerator1
    Position Offset=<0 0 -22>
    Sprite Rotation Mode=::NewEnumerator2
    Sprite Size Mode=::NewEnumerator1
    Uniform Sprite Size Max=40
    Uniform Sprite Size Min=22
    UsePositionOffset=true
    Write Material Random=true
   )
   &ShapeLocation(Box Size=<42 42 10>, MaskSpawnGroup=false, Override Seed=false, Shape Primitive=::NewEnumerator2, Surface Only=false, UseOwnerScale=true)
   &CameraOffset(Camera Offset Amount=10)
  ]
  particleUpdateScript [
   &ParticleState()
   &ScaleSpriteSize(Uniform Curve Sprite Scale=DataInterface() {propertyValues {Curve {keys [{tangentMode=RCTM_User, time=0, value=0.4}, <at 0.2 1>, {tangentMode=RCTM_User, time=1, value=1}]}}})
   &ScaleColor(
    Scale Alpha=&FloatFromCurve(FloatCurve=DataInterface() {propertyValues {Curve {keys [
         <at 0 0 RCIM_Cubic>
         {arriveTangent=2, interpMode=RCIM_Cubic, leaveTangent=0.0238, tangentMode=RCTM_Break, time=0.1, value=1}
         {interpMode=RCIM_Cubic, leaveTangent=-3.1814, tangentMode=RCTM_Break, time=0.5, value=1}
         <at 1 0>
        ]}}})
    ScaleA=true
    ScaleRGB=true
   )
   &GravityForce(Gravity=<0 0 -15>)
   &SolveForcesAndVelocity(Clamp Velocity=false, Limit Acceleration=false)
  ]
  rendererClasses [&NiagaraSpriteRendererProperties]
  renderers [{rendererClass=&NiagaraSpriteRendererProperties}]
 }
 Spotlight {
  emitterUpdateScript [
   &EmitterState(
    Inactive Response=::NewEnumerator1
    Life Cycle Mode=::NewEnumerator1
    Loop Behavior=::NewEnumerator1
    Loop Duration Mode=::NewEnumerator1
    UseLoopDelay=false
   )
   &SpawnBurst_Instantaneous(Use Spawn Probability=false, UseLoopCountLimit=false)
  ]
  particleSpawnScript [
   &InitializeParticle(
    Color=&FNTimeOfDayColor(
     Night Color=<rgba 0.5 0.4132 0.25 0.6>
     Morning Color=<rgba 0.5 0.4132 0.25 0.6>
     Day Color=&MakeCustomLinearColorFromBool(
      Bool=&Boolean_OrOperation(Bool1=$User.bIsFrontEnd, Bool2=$User.bIsFrontEndPreview)
      True Color=<rgba 0.8 0.632 0.24 0.275>
      False Color=<rgba 1 0.6877 0.2311 0.5>
     )
     Evening Color=<rgba 0.5 0.4132 0.25 0.6>
    )
    Lifetime Max=0.6
    Lifetime Min=0.8
    Lifetime Mode=::NewEnumerator1
    Mesh Scale=<0.6 0.6 0.8>
    Mesh Scale Mode=::NewEnumerator4
    Position Offset=<0 0 140>
    Sprite Size Mode=::NewEnumerator3
    Uniform Sprite Size=77
    UsePositionOffset=true
    Write Material Random=true
   )
   &InitialMeshOrientation(Mesh Orientation Mode=::NewEnumerator7, Use Rotation Vector=true)
   /SetVariables(Particles.GlowCameraOffset=33) {isSetParameters=true}
  ]
  particleUpdateScript [
   &ParticleState(Kill Particles When Lifetime Has Elapsed=false, Let Infinitely Lived Particles Die When Emitter Deactivates=true)
   &ScaleMeshSize(Scale Factor=&VectorFromCurve(
     VectorCurve=DataInterface() {propertyValues {
       XCurve {keys [<at 0 0>,<at 0.01 0.8>,<at 0.99 0.8>,<at 1 0>]}
       YCurve {keys [<at 0 0>,<at 0.01 0.8>,<at 0.99 0.8>,<at 1 0>]}
       ZCurve {keys [<at 0 1>,<at 1 1>]}
      }}
     CurveIndex=$User.NormalizedNotifyProgress
    )) {enabled=false, module=ScaleMeshSize001}
   &ScaleSpriteSize(
    Uniform Curve Index=$User.NormalizedNotifyProgress
    Uniform Curve Sprite Scale=DataInterface() {propertyValues {Curve {keys [{tangentMode=RCTM_User, time=0, value=0}, {tangentMode=RCTM_User, time=0.03, value=1}]}}}
   ) {module=ScaleSpriteSize001}
   /SetVariables(Particles.GlowColor=&MultiplyLinearColor(Linear Color=$Particles.Color, Scale Factor=&MakeVector4_FromFloat(Float=0.5))) {isSetParameters=true}
   &ScaleColor(
    Scale Alpha=&FloatFromCurve(
     FloatCurve=DataInterface() {propertyValues {Curve {keys [
         <at 0.2 0>
         <at 0.3 1>
         {interpMode=RCIM_Cubic, leaveTangent=-0.8016, tangentMode=RCTM_Break, time=0.8, value=1}
         {arriveTangent=-8.0033, tangentMode=RCTM_User, time=1, value=0}
        ]}}}
     CurveIndex=$User.NormalizedNotifyProgress
    )
    ScaleA=true
    ScaleRGB=true
   )
   /SetVariables() {isSetParameters=true}
  ]
  rendererClasses [&NiagaraMeshRendererProperties]
  renderers [
   {rendererClass=&NiagaraMeshRendererProperties, rendererIndex=0}
   {rendererClass=&NiagaraMeshRendererProperties, rendererIndex=1}
  ]
 }
}
summary {
 systemName=NS_Emote_FeatherMud_Looping
 emitters [
  {emitterName=Confetti, rendererClasses [&NiagaraSpriteRendererProperties]}
  {emitterName=Spotlight, rendererClasses [&NiagaraMeshRendererProperties]}
  {emitterName=MagicalPuffCloud, rendererClasses [&NiagaraSpriteRendererProperties]}
 ]
 userVariables [
  User.bIsFrontEnd {type=NiagaraBool}
  User.bIsFrontEndPreview {type=NiagaraBool}
  User.NormalizedNotifyProgress {type=NiagaraFloat, defaultValue=0, description=""}
 ]
}
